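# Physics asset for the cat pet backpack mesh, simulating a dangling yarn piece on the cosmetic back bling.
# A large capsule on the root bone stands in for the pack body, while two small capsules cover the yarn bones.
# Angular-limited constraints chain the first yarn bone to root and the second to the first, giving swing with restricted twist.
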PhysicsAsset /BRCosmetics/Accessories/FORT_Backpacks/Backpack_Pet_Cat/Meshes/M_MED_CatPack_Physics
bodies {
 dyn_yarn_01 [{shapeType=Capsule, length=1.2271, radius=1, center=<1.925 0.0001 0>, rotation=<rot 90 0 0>}]
 dyn_yarn_02 [{shapeType=Capsule, length=0.2705, radius=4.6982, center=<4.3241 0.452 0.7161>, rotation=<rot 89.8991 0 0>}]
 root [{shapeType=Capsule, length=11.6461, radius=7.3257, center=<0 -5.1335 -5.9247>}]
}
constraints [
 {bone1Name=dyn_yarn_01, bone2Name=root, swing1LimitDegrees=45, swing2LimitDegrees=45, twistLimitDegrees=15}
 {bone1Name=dyn_yarn_02, bone2Name=dyn_yarn_01, swing1LimitDegrees=45, swing2LimitDegrees=45, twistLimitDegrees=10}
]
