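# Animation Blueprint for the cat pet backpack skeleton that adds procedural secondary motion on top of an incoming pose.
# Two chained root-relative Anim Dynamics nodes with reduced gravity move the yarn bone within linear limits and the crown bone within small angular limits.
# A RigidBody node using the CatPack physics asset then simulates on that result before output.
# All work happens in the AnimGraph, with no variables or event logic.
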AnimBlueprint /BRCosmetics/Accessories/FORT_Backpacks/Backpack_Pet_Cat/Meshes/M_MED_CatPack_Skeleton_AnimBP
target_skeleton=&/M_MED_CatPack_Skeleton
graphs [
 AnimGraph {nodes [
   AnimGraphNode_SubInput_1 = Input Pose()
   n1 = Anim Dynamics - Bone: dyn_yarn_03(ComponentPose=Local To Component(LocalPose=@AnimGraphNode_SubInput_1), Alpha=1, ExternalForce="0,0,0") {
    bGravityOverrideInSimSpace=true
    bOverrideLinearDamping=true
    gravityScale=0.32
    lODThreshold=1
    linearDampingOverride=0.9
    simulationSpace=RootRelative
    alphaBoolBlend {blendOption=Linear}
    boundBone {boneName=dyn_yarn_03}
    componentLinearAccScale=<1 1 1>
    componentLinearVelScale=<1 1 1>
    physicsBodyDefinitions [{
      sphereCollisionRadius=0
      boundBone {boneName=dyn_yarn_03}
      boxExtents=<2 2 2>
      constraintSetup {angularLimitsMax=<0 0 0>, angularLimitsMin=<0 0 0>, angularTarget=<0 0 0>, linearAxesMax=<0 1 1>, linearAxesMin=<-3 -1 -1>}
      localJointOffset=<-0.1 0 0>
     }]
    relativeSpaceBone {boneName=root}
   }
   n2 = Anim Dynamics - Bone: dyn_crown(ComponentPose=@n1, Alpha=1, ExternalForce="0,0,0") {
    bGravityOverrideInSimSpace=true
    gravityScale=0.32
    simulationSpace=RootRelative
    alphaBoolBlend {blendOption=Linear}
    boundBone {boneName=dyn_crown}
    componentLinearAccScale=<1 1 1>
    physicsBodyDefinitions [{
      sphereCollisionRadius=0
      boundBone {boneName=dyn_crown}
      boxExtents=<5 5 5>
      constraintSetup {angularLimitsMax=<3 3 5>, angularLimitsMin=<-3 -3 -5>, angularTarget=<0 0 0>}
      localJointOffset=<-0.1 0 0>
     }]
    relativeSpaceBone {boneName=root}
   }
   n3 = RigidBody(ComponentPose=@n2) {
    overridePhysicsAsset=&/M_MED_CatPack_Physics
    alphaBoolBlend {blendOption=Linear}
    baseBoneRef {boneName=root}
    componentAppliedLinearAccClamp=<500 500 500>
    componentLinearAccScale=<1 1 1>
    componentLinearVelScale=<1 1 1>
   }
   Output Pose(Result=Component To Local(ComponentPose=@n3)) {layerGroup=None}
  ]}
 EventGraph {nodes [Event BlueprintUpdateAnimation()]}
]
variables []
